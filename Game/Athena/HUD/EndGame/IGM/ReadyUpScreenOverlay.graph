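# Post-game ready-up overlay for starting the next match, with party slots, buttons and a status switcher for countdown, searching, waiting and error.
# Native AthenaReadyUpScreenOverlay events drive sounds and a clipped countdown that scrolls by animating text padding.
# Matchmaking start hides tagged HUD elements until Destruct; successful completion restarts a timer for delayed audio, apparently the match flush.
# The error header's text is cached at PreConstruct as a format template for the exit-to-lobby seconds.
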WidgetBlueprint /Game/Athena/HUD/EndGame/IGM/ReadyUpScreenOverlay
animations [
 Countdown {
  binding_count=2
  length_seconds=0.2628
  object=&self:Countdown
  sequence {bindings [
    Text_CountdownCurrent {id=#1, possessed_class=CommonTextBlock, tracks []}
    VerticalBoxSlot_0 {
     id=#2
     parent_id=#1
     possessed_class=VerticalBoxSlot
     tracks [{
       type=MovieSceneMarginTrack
       display_name=Padding
       sections [{
         type=MovieSceneMarginSection
         channels [
          Left {type=MovieSceneScriptingFloatChannel}
          Top {type=MovieSceneScriptingFloatChannel, keys [[0,-15],[5.2567,-180]]}
          Right {type=MovieSceneScriptingFloatChannel}
          Bottom {type=MovieSceneScriptingFloatChannel}
         ]
        }]
      }]
    }
   ]}
 }
 InitialCountdownState {
  binding_count=2
  length_seconds=0.05
  object=&self:InitialCountdownState
  sequence {bindings [
    VerticalBoxSlot_0 {
     id=#3
     parent_id=#4
     possessed_class=VerticalBoxSlot
     tracks [{
       type=MovieSceneMarginTrack
       display_name=Padding
       sections [{
         type=MovieSceneMarginSection
         end_frame=0
         start_frame=0
         channels [
          Left {type=MovieSceneScriptingFloatChannel, const [0,1]}
          Top {type=MovieSceneScriptingFloatChannel, const [-15,1]}
          Right {type=MovieSceneScriptingFloatChannel, const [0,1]}
          Bottom {type=MovieSceneScriptingFloatChannel, const [-30,1]}
         ]
        }]
      }]
    }
    Text_CountdownCurrent {id=#4, possessed_class=CommonTextBlock, tracks []}
   ]}
 }
]
graphs [
 EventGraph {nodes [
   PlayFlushAudio() {is=title K2Node_CustomEvent}
   > SpawnSound2D!(
    execute
    Sound:Sound Base Object Reference=Get MatchFlush() -> (MatchFlush:Sound Base Object Reference)
    VolumeMultiplier:float=1
    PitchMultiplier:float=1
    StartTime:float=0
    ConcurrencySettings:Sound Concurrency Object Reference
    bPersistAcrossLevelTransition:bool=true
    bAutoDestroy:bool=true
   ) -> (then, ReturnValue:Audio Component Object Reference)
   Event PreConstruct()
   > Set CachedErrorText(CachedErrorText:Text="GetText (Text)"!(self:Text Object Reference=Get Text_ErrorHeader() -> (Text_ErrorHeader:Common Text Object Reference)) -> (ReturnValue:Text)) -> (Output_Get:Text)
   Event OnCountdownProgressed() {is=title K2Node_Event}
   > PlayAnimation(InAnimation=Get Countdown() -> (Countdown:Widget Animation Object Reference))
   > PlaySound2D(Sound=Get TimeTick() -> (TimeTick:Sound Base Object Reference))
   Event OnReadyUp() {is=title K2Node_Event}
   > SpawnSound2D(
    Sound=Get ReadyUp() -> (ReadyUp:Sound Base Object Reference)
    VolumeMultiplier=1
    PitchMultiplier=1
    StartTime=0
    bPersistAcrossLevelTransition=true
    bAutoDestroy=true
   )
   > PlayAnimation(InAnimation=Get InitialCountdownState() -> (InitialCountdownState:Widget Animation Object Reference))
   Event OnCancelReadyUp() {is=title K2Node_Event}
   > SpawnSound2D(
    Sound=Get Unready() -> (Unready:Sound Base Object Reference)
    VolumeMultiplier=1
    PitchMultiplier=1
    StartTime=0
    bPersistAcrossLevelTransition=true
    bAutoDestroy=true
   )
   Event OnMatchmakingStarted() {is=title K2Node_Event}
   > SpawnSound2D(
    Sound=Get MatchInitiated() -> (MatchInitiated:Sound Base Object Reference)
    VolumeMultiplier=1
    PitchMultiplier=1
    StartTime=0
    bPersistAcrossLevelTransition=true
    bAutoDestroy=true
   )
   > n1 = Cast To FortPlayerControllerAthena(Object=GetOwningPlayer())
   > SetHUDElementVisibilityWithContext!(
    execute
    self:Fort Player Controller Object Reference=@n1
    HUDElementTags:"Gameplay Tag Container Structure (by ref)"=MakeGameplayTagContainerFromTag!(SingleTag:Gameplay Tag Structure=(TagName="Athena.Papaya.HideForPostGame")) -> (ReturnValue:Gameplay Tag Container Structure)
    bHideElements:bool=true
    ReasonName:Name=None
    LogContext:String
   ) -> (then)
   n2 = Event OnExitToLobbyCountdownProgressed!() -> (OutputDelegate, then, CountdownValue:float) {is=title K2Node_Event}
   > UpdateExitToLobbyCountdownText!(execute, self, CountdownValue:double=@n2.CountdownValue) -> (then)
   n3 = Event OnMatchmakingFinished!() -> (OutputDelegate, then, bSuccess:bool) {is=title K2Node_Event}
   > Branch(Condition=@n3.bSuccess)
   n4 = Get FlushAudioTimer() -> (FlushAudioTimer:Timer Handle Structure)
   > n5 = Branch(Condition=Is Valid Timer Handle!(Handle:Timer Handle Structure=@n4) -> (ReturnValue:bool))
   > n6 = Clear and Invalidate Timer by Handle(Handle=@n4)
   n7 = Set Timer by Event(execute=[@n6.then, @n5.else], Delegate=Create Event(), Time=Get Time() -> (Time:float), bLooping=false)
   > Set FlushAudioTimer(FlushAudioTimer:Timer Handle Structure=@n7) -> (Output_Get:Timer Handle Structure)
   Event OnCountdownStarted() {is=title K2Node_Event}
   > PlayAnimation(InAnimation=Get InitialCountdownState())
   Event Destruct()
   > n8 = Cast To FortPlayerControllerAthena(Object=GetOwningPlayer())
   > SetHUDElementVisibilityWithContext(
    self=@n8
    HUDElementTags=MakeGameplayTagContainerFromTag(SingleTag=(TagName="Athena.Papaya.HideForPostGame"))
    bHideElements=false
    ReasonName=None
   )
  ]}
 UpdateExitToLobbyCountdownText {nodes [
   n1 = UpdateExitToLobbyCountdownText!() -> (then, CountdownValue:double=~) {is=title K2Node_FunctionEntry}
   > "SetText (Text)"(
    self=Get Text_ErrorHeader()
    InText=Format Text!(Format=Get CachedErrorText() -> (CachedErrorText:Text), NUMBEROFSECONDS:int=Round(A=@n1) {is=title K2Node_CallFunction}) -> (Result)
   )
  ]}
]
variables [
 CachedErrorText {type=string, default=""}
 FlushAudioTimer {type=TimerHandle}
 MatchFlush {type=SoundBase, default=/Game/Sounds/Fort_UI_Sounds/TrayNotifications/SC_TrayNotification_MatchFlush}
 MatchInitiated {type=SoundBase, default=/Game/Sounds/Fort_UI_Sounds/FrontEnd/SC_UI_MatchMaking_MatchInitiated_Cue}
 ReadyUp {type=SoundBase, default=/Game/Sounds/Fort_UI_Sounds/FrontEnd/SC_UI_MatchMaking_ReadyUp_Cue}
 Time {type=number, default=4.1}
 TimeTick {type=SoundBase, default=/Game/Sounds/Fort_UI_Sounds/FrontEnd/SC_UI_MatchMaking_TimeTick_Cue}
 Unready {type=SoundBase, default=/Game/Sounds/Fort_UI_Sounds/FrontEnd/SC_UI_MatchMaking_UnReady_Cue}
]
widgets {
 info {inheritedWidgetCount=15, parentClass=/Script/FortniteUI.AthenaReadyUpScreenOverlay, rootWidgetClass=&Overlay, widgetCount=48}
 widgets [
  Overlay_1 {widgetClassPath=&Overlay, widget=Overlay_1}
  Text_PartyKPTStatus {
   widgetClassPath=&CommonTextBlock
   bInherited=true
   parent=Overlay_1
   widget=Text_PartyKPTStatus
   properties {
    text=Your party left the team
    textTransformPolicy=ToUpper
    visibility=Collapsed
    font {
     bForceMonospaced=false
     bMaterialIsStencil=false
     fontMaterial=null
     fontObject=&Burbank
     letterSpacing=0
     monospacedWidth=1
     size=34
     skewAmount=0
     typefaceFontName=Burbank-PlayerPower
     outlineSettings {
      bApplyOutlineToDropShadows=false
      bMiteredCorners=false
      bSeparateFillAlpha=false
      outlineMaterial=null
      outlineSize=0
      outlineColor=<rgba 0 0 0 1>
     }
    }
    renderTransform {angle=0, scale=<1 1>, shear=<-8 0>, translation=<0 0>}
    shadowOffset=<0 0>
   }
   slot {class=OverlaySlot, properties {horizontalAlignment=HAlign_Center, padding=<ltrb 0 200 0 0>}}
  }
  VerticalBox_0 {
   widgetClassPath=&VerticalBox
   parent=Overlay_1
   widget=VerticalBox_0
   slot {class=OverlaySlot, properties {horizontalAlignment=HAlign_Fill, verticalAlignment=VAlign_Fill}}
  }
  Overlay_0 {
   widgetClassPath=&Overlay
   parent=VerticalBox_0
   widget=Overlay_0
   slot {class=VerticalBoxSlot, properties {padding=<ltrb 0 0 0 20>, size {sizeRule=Fill, value=1}}}
  }
  SocialUserBox {
   widgetClassPath=&HorizontalBox
   bIsVariable=true
   parent=Overlay_0
   widget=SocialUserBox
   slot {class=OverlaySlot, properties {horizontalAlignment=HAlign_Center, verticalAlignment=VAlign_Bottom}}
  }
  LocalPlayerSlot {
   widgetClassPath=&/InGameSocialUserWidget
   bIsVariable=true
   parent=SocialUserBox
   widget=LocalPlayerSlot
   properties {bIsLocalPlayerSlot=true, visibility=Collapsed}
   slot {class=HorizontalBoxSlot, properties {padding=<ltrb 12 0 12 0>}}
  }
  InGameSlot_0 {
   widgetClassPath=&InGameSocialUserWidget
   bIsVariable=true
   parent=SocialUserBox
   widget=InGameSlot_0
   properties {notPlayer=true, visibility=Collapsed}
   slot {class=HorizontalBoxSlot, properties {horizontalAlignment=HAlign_Left, verticalAlignment=VAlign_Center, padding=<ltrb 12 0 12 0>}}
  }
  InGameSlot_1 {
   widgetClassPath=&InGameSocialUserWidget
   bIsVariable=true
   parent=SocialUserBox
   widget=InGameSlot_1
   properties {notPlayer=true, visibility=Collapsed}
   slot {class=HorizontalBoxSlot, properties {horizontalAlignment=HAlign_Left, verticalAlignment=VAlign_Center, padding=<ltrb 25 0 12 0>}}
  }
  InGameSlot_2 {
   widgetClassPath=&InGameSocialUserWidget
   bIsVariable=true
   parent=SocialUserBox
   widget=InGameSlot_2
   properties {notPlayer=true, visibility=Collapsed}
   slot {class=HorizontalBoxSlot, properties {horizontalAlignment=HAlign_Left, verticalAlignment=VAlign_Center, padding=<ltrb 25 0 12 0>}}
  }
  VerticalBox_70 {
   widgetClassPath=&VerticalBox
   parent=Overlay_0
   widget=VerticalBox_70
   slot {class=OverlaySlot, properties {verticalAlignment=VAlign_Center}}
  }
  LobbySlot_0 {
   widgetClassPath=&/InLobbySocialUserWidget
   bIsVariable=true
   parent=VerticalBox_70
   widget=LobbySlot_0
   properties {visibility=Collapsed}
   slot {class=VerticalBoxSlot}
  }
  LobbySlot_1 {
   widgetClassPath=&InLobbySocialUserWidget
   bIsVariable=true
   parent=VerticalBox_70
   widget=LobbySlot_1
   properties {visibility=Collapsed}
   slot {class=VerticalBoxSlot}
  }
  LobbySlot_2 {
   widgetClassPath=&InLobbySocialUserWidget
   bIsVariable=true
   parent=VerticalBox_70
   widget=LobbySlot_2
   properties {visibility=Collapsed}
   slot {class=VerticalBoxSlot}
  }
  VerticalBox_1 {
   widgetClassPath=&VerticalBox
   parent=Overlay_0
   widget=VerticalBox_1
   slot {class=OverlaySlot, properties {horizontalAlignment=HAlign_Right, verticalAlignment=VAlign_Bottom, padding=<ltrb 10 10 28 15>}}
  }
  Button_StartMatch {
   widgetClassPath=/FortUILibrary/UIKit/Buttons/WBP_UIKit_Button_Loud
   bInherited=true
   bIsVariable=true
   parent=VerticalBox_1
   widget=Button_StartMatch
   properties {buttonSize=Large, text=START MATCH, triggeringInputAction {dataTable=&InputActionDataTable, rowName=CustomStartMatch}}
   slot {class=VerticalBoxSlot, properties {horizontalAlignment=HAlign_Center}}
  }
  Switcher_MatchmakingStatus {
   widgetClassPath=&CommonWidgetSwitcherLegacy
   bInherited=true
   bIsVariable=true
   parent=VerticalBox_1
   widget=Switcher_MatchmakingStatus
   slot {class=VerticalBoxSlot}
  }
  Container_CountdownStatus {
   widgetClassPath=&HorizontalBox
   bInherited=true
   parent=Switcher_MatchmakingStatus
   widget=Container_CountdownStatus
   slot {class=WidgetSwitcherSlot}
  }
  SizeBox_1 {
   widgetClassPath=&SizeBox
   parent=Container_CountdownStatus
   widget=SizeBox_1
   properties {bOverride_WidthOverride=true, maxDesiredWidth=303.0232, widthOverride=290}
   slot {class=HorizontalBoxSlot, properties {horizontalAlignment=HAlign_Right, verticalAlignment=VAlign_Center, padding=<ltrb 0 0 20 0>}}
  }
  ScaleBox_0 {widgetClassPath=&ScaleBox, parent=SizeBox_1, widget=ScaleBox_0, properties {stretchDirection=DownOnly}, slot {class=SizeBoxSlot}}
  Text_CountdownStatus {
   widgetClassPath=&CommonTextBlock
   bInherited=true
   parent=ScaleBox_0
   widget=Text_CountdownStatus
   properties {
    justification=Right
    text="Next Match\r\nStarting in"
    textTransformPolicy=ToUpper
    font {
     bForceMonospaced=false
     bMaterialIsStencil=false
     fontMaterial=null
     fontObject=&Burbank
     letterSpacing=0
     monospacedWidth=1
     size=35
     skewAmount=0
     typefaceFontName=Burbank-PlayerPower
     outlineSettings {
      bApplyOutlineToDropShadows=false
      bMiteredCorners=false
      bSeparateFillAlpha=false
      outlineMaterial=null
      outlineSize=0
      outlineColor=<rgba 0 0 0 1>
     }
    }
    renderTransform {angle=0, scale=<1 1>, shear=<-8 0>, translation=<0 0>}
    shadowOffset=<0 0>
   }
   slot {class=ScaleBoxSlot, properties {horizontalAlignment=HAlign_Right}}
  }
  SizeBox_0 {
   widgetClassPath=&SizeBox
   parent=Container_CountdownStatus
   widget=SizeBox_0
   properties {
    bOverride_HeightOverride=true
    bOverride_MaxDesiredWidth=true
    bOverride_WidthOverride=true
    clipping=ClipToBounds
    heightOverride=165
    maxDesiredWidth=210
    widthOverride=233.3191
    renderTransform {angle=0, scale=<1 1>, shear=<-12 0>, translation=<0 0>}
   }
   slot {class=HorizontalBoxSlot, properties {verticalAlignment=VAlign_Bottom, padding=<ltrb -20 0 -20 0>}}
  }
  Container_Countdown {
   widgetClassPath=&VerticalBox
   parent=SizeBox_0
   widget=Container_Countdown
   slot {class=SizeBoxSlot, properties {horizontalAlignment=HAlign_Center, verticalAlignment=VAlign_Center}}
  }
  Text_CountdownCurrent {
   widgetClassPath=&CommonTextBlock
   bInherited=true
   parent=Container_Countdown
   widget=Text_CountdownCurrent
   properties {
    justification=Center
    text=15
    textTransformPolicy=ToUpper
    font {
     bForceMonospaced=false
     bMaterialIsStencil=false
     fontMaterial=null
     fontObject=&Burbank
     letterSpacing=0
     monospacedWidth=1
     size=126
     skewAmount=0
     typefaceFontName=Burbank-PlayerPower
     outlineSettings {
      bApplyOutlineToDropShadows=false
      bMiteredCorners=false
      bSeparateFillAlpha=false
      outlineMaterial=null
      outlineSize=0
      outlineColor=<rgba 0 0 0 1>
     }
    }
    shadowOffset=<0 0>
   }
   slot {class=VerticalBoxSlot, properties {padding=<ltrb 0 -15 0 -30>}}
  }
  Text_CountdownNext {
   widgetClassPath=&CommonTextBlock
   bInherited=true
   parent=Container_Countdown
   widget=Text_CountdownNext
   properties {
    justification=Center
    text=14
    textTransformPolicy=ToUpper
    font {
     bForceMonospaced=false
     bMaterialIsStencil=false
     fontMaterial=null
     fontObject=&Burbank
     letterSpacing=0
     monospacedWidth=1
     size=126
     skewAmount=0
     typefaceFontName=Burbank-PlayerPower
     outlineSettings {
      bApplyOutlineToDropShadows=false
      bMiteredCorners=false
      bSeparateFillAlpha=false
      outlineMaterial=null
      outlineSize=0
      outlineColor=<rgba 0 0 0 1>
     }
    }
    shadowOffset=<0 0>
   }
   slot {class=VerticalBoxSlot, properties {padding=<ltrb 0 -15 0 -30>}}
  }
  Container_FindingMatchStatus {
   widgetClassPath=&VerticalBox
   bInherited=true
   parent=Switcher_MatchmakingStatus
   widget=Container_FindingMatchStatus
   slot {class=WidgetSwitcherSlot}
  }
  HBox {widgetClassPath=&HorizontalBox, parent=Container_FindingMatchStatus, widget=HBox, slot {class=VerticalBoxSlot}}
  SizeBox {
   widgetClassPath=&SizeBox
   parent=HBox
   widget=SizeBox
   properties {bOverride_WidthOverride=true, maxDesiredWidth=303.0232, widthOverride=290}
   slot {class=HorizontalBoxSlot, properties {horizontalAlignment=HAlign_Right, verticalAlignment=VAlign_Center, padding=<ltrb 0 0 20 0>}}
  }
  ScaleBox {widgetClassPath=&ScaleBox, parent=SizeBox, widget=ScaleBox, properties {stretchDirection=DownOnly}, slot {class=SizeBoxSlot}}
  Text_MatchmakingMessage {
   widgetClassPath=&CommonTextBlock
   bInherited=true
   parent=ScaleBox
   widget=Text_MatchmakingMessage
   properties {
    justification=Right
    text="Finding\r\nNext Match"
    textTransformPolicy=ToUpper
    font {
     bForceMonospaced=false
     bMaterialIsStencil=false
     fontMaterial=null
     fontObject=&Burbank
     letterSpacing=0
     monospacedWidth=1
     size=35
     skewAmount=0
     typefaceFontName=Burbank-PlayerPower
     outlineSettings {
      bApplyOutlineToDropShadows=false
      bMiteredCorners=false
      bSeparateFillAlpha=false
      outlineMaterial=null
      outlineSize=0
      outlineColor=<rgba 0 0 0 1>
     }
    }
    renderTransform {angle=0, scale=<1 1>, shear=<-8 0>, translation=<0 0>}
    shadowOffset=<0 0>
   }
   slot {class=ScaleBoxSlot, properties {horizontalAlignment=HAlign_Right}}
  }
  SizeBox_2 {
   widgetClassPath=&SizeBox
   parent=HBox
   widget=SizeBox_2
   properties {
    bOverride_HeightOverride=true
    bOverride_MaxDesiredWidth=true
    bOverride_WidthOverride=true
    clipping=ClipToBounds
    heightOverride=165
    maxDesiredWidth=210
    widthOverride=233.3191
    renderTransform {angle=0, scale=<1 1>, shear=<-12 0>, translation=<0 0>}
   }
   slot {class=HorizontalBoxSlot, properties {horizontalAlignment=HAlign_Left, verticalAlignment=VAlign_Bottom, padding=<ltrb -20 0 -20 0>}}
  }
  ScaleBox_122 {
   widgetClassPath=&ScaleBox
   parent=SizeBox_2
   widget=ScaleBox_122
   properties {stretchDirection=DownOnly}
   slot {class=SizeBoxSlot, properties {horizontalAlignment=HAlign_Center, padding=<ltrb 0 0 5 0>}}
  }
  Text_MatchmakingElapsed {
   widgetClassPath=&CommonTextBlock
   bInherited=true
   parent=ScaleBox_122
   widget=Text_MatchmakingElapsed
   properties {
    justification=Center
    text=00:00
    textTransformPolicy=ToUpper
    font {
     bForceMonospaced=false
     bMaterialIsStencil=false
     fontMaterial=null
     fontObject=&Burbank
     letterSpacing=0
     monospacedWidth=1
     size=60
     skewAmount=0
     typefaceFontName=Burbank-PlayerPower
     outlineSettings {
      bApplyOutlineToDropShadows=false
      bMiteredCorners=false
      bSeparateFillAlpha=false
      outlineMaterial=null
      outlineSize=0
      outlineColor=<rgba 0 0 0 1>
     }
    }
    shadowOffset=<0 0>
   }
   slot {class=ScaleBoxSlot}
  }
  Container_Waiting {
   widgetClassPath=&VerticalBox
   bInherited=true
   parent=Switcher_MatchmakingStatus
   widget=Container_Waiting
   slot {class=WidgetSwitcherSlot}
  }
  Vbox_1 {widgetClassPath=&VerticalBox, parent=Container_Waiting, widget=Vbox_1, slot {class=VerticalBoxSlot}}
  SizeBox_4 {
   widgetClassPath=&SizeBox
   parent=Vbox_1
   widget=SizeBox_4
   properties {bOverride_WidthOverride=true, maxDesiredWidth=303.0232, widthOverride=500}
   slot {class=VerticalBoxSlot}
  }
  ScaleBox_2 {
   widgetClassPath=&ScaleBox
   parent=SizeBox_4
   widget=ScaleBox_2
   properties {stretchDirection=DownOnly}
   slot {class=SizeBoxSlot, properties {padding=<ltrb 20 0 0 0>}}
  }
  Text_Waiting {
   widgetClassPath=&CommonTextBlock
   bIsVariable=true
   parent=ScaleBox_2
   widget=Text_Waiting
   properties {
    justification=Right
    text="Waiting for\r\n party to ready up..."
    textTransformPolicy=ToUpper
    font {
     bForceMonospaced=false
     bMaterialIsStencil=false
     fontMaterial=null
     fontObject=&Burbank
     letterSpacing=0
     monospacedWidth=1
     size=35
     skewAmount=0
     typefaceFontName=Burbank-PlayerPower
     outlineSettings {
      bApplyOutlineToDropShadows=false
      bMiteredCorners=false
      bSeparateFillAlpha=false
      outlineMaterial=null
      outlineSize=0
      outlineColor=<rgba 0 0 0 1>
     }
    }
    renderTransform {angle=0, scale=<1 1>, shear=<-8 0>, translation=<0 0>}
    shadowOffset=<0 0>
   }
   slot {class=ScaleBoxSlot, properties {horizontalAlignment=HAlign_Right}}
  }
  Container_Error {
   widgetClassPath=&VerticalBox
   bInherited=true
   parent=Switcher_MatchmakingStatus
   widget=Container_Error
   slot {class=WidgetSwitcherSlot}
  }
  Vbox {widgetClassPath=&VerticalBox, parent=Container_Error, widget=Vbox, slot {class=VerticalBoxSlot}}
  SizeBox_3 {
   widgetClassPath=&SizeBox
   parent=Vbox
   widget=SizeBox_3
   properties {bOverride_WidthOverride=true, maxDesiredWidth=303.0232, widthOverride=500}
   slot {class=VerticalBoxSlot}
  }
  ScaleBox_1 {
   widgetClassPath=&ScaleBox
   parent=SizeBox_3
   widget=ScaleBox_1
   properties {stretchDirection=DownOnly}
   slot {class=SizeBoxSlot, properties {padding=<ltrb 20 0 0 0>}}
  }
  Text_ErrorHeader {
   widgetClassPath=&CommonTextBlock
   bIsVariable=true
   parent=ScaleBox_1
   widget=Text_ErrorHeader
   properties {
    justification=Right
    text="MATCHMAKING FAILED\r\nRETURNING TO THE \r\nLOBBY IN {NUMBEROFSECONDS} SECONDS"
    textTransformPolicy=ToUpper
    font {
     bForceMonospaced=false
     bMaterialIsStencil=false
     fontMaterial=null
     fontObject=&Burbank
     letterSpacing=0
     monospacedWidth=1
     size=35
     skewAmount=0
     typefaceFontName=Burbank-PlayerPower
     outlineSettings {
      bApplyOutlineToDropShadows=false
      bMiteredCorners=false
      bSeparateFillAlpha=false
      outlineMaterial=null
      outlineSize=0
      outlineColor=<rgba 0 0 0 1>
     }
    }
    renderTransform {angle=0, scale=<1 1>, shear=<-8 0>, translation=<0 0>}
    shadowOffset=<0 0>
   }
   slot {class=ScaleBoxSlot, properties {horizontalAlignment=HAlign_Right}}
  }
  Container_NoStatus {
   widgetClassPath=&HorizontalBox
   bInherited=true
   parent=Switcher_MatchmakingStatus
   widget=Container_NoStatus
   slot {class=WidgetSwitcherSlot}
  }
  AthenaServerShutdownWidget {
   widgetClassPath=&/AthenaServerShutdownWidget
   bIsVariable=true
   parent=Overlay_0
   widget=AthenaServerShutdownWidget
   slot {class=OverlaySlot, properties {padding=<ltrb 15 10 15 15>}}
  }
  Overlay_2 {widgetClassPath=&Overlay, parent=VerticalBox_0, widget=Overlay_2, slot {class=VerticalBoxSlot}}
  PostGameBottomBarWidget {
   widgetClassPath=/Game/Athena/HUD/EndGame/PostGameBottomBarWidget
   bIsVariable=true
   parent=Overlay_2
   widget=PostGameBottomBarWidget
   slot {class=OverlaySlot, properties {horizontalAlignment=HAlign_Fill, verticalAlignment=VAlign_Fill}}
  }
  Button_Cancel {
   widgetClassPath=/FortUILibrary/UIKit/Buttons/WBP_UIKit_Button_Quiet
   bInherited=true
   bIsVariable=true
   parent=Overlay_2
   widget=Button_Cancel
   properties {text=Cancel, triggeringInputAction {dataTable=&InputActionDataTable, rowName=GenericFaceButtonRight}}
   slot {class=OverlaySlot, properties {verticalAlignment=VAlign_Center, padding=<ltrb 32 16 16 16>}}
  }
  Button_StayTogether {
   widgetClassPath=&/StayTogetherButton
   bInherited=true
   parent=Overlay_2
   widget=Button_StayTogether
   slot {class=OverlaySlot, properties {horizontalAlignment=HAlign_Right, verticalAlignment=VAlign_Center, padding=<ltrb 16 16 32 16>}}
  }
 ]
}
