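# Widget Blueprint for the default post-game flow Ready Up screen, derived from the native AthenaReadyUpScreenWidget class.
# A root overlay stacks a transparent fill spacer image, a collapsed "READY UP SCREEN" label, and two full-fill child widgets.
# Those children, ReadyUpOverlay and KeepPlayingTogetherOverlay, are the inherited widgets from the end-game IGM folder.
# The event graph is empty and no variables are declared, so the file supplies layout only.
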WidgetBlueprint /Game/Athena/HUD/EndGame/PostGameFlowScreens/DefaultScreens/ReadyUpScreen
graphs [EventGraph {links [], nodes []}]
variables []
widgets {
 info {inheritedWidgetCount=2, parentClass=/Script/FortniteUI.AthenaReadyUpScreenWidget, rootWidgetClass=&Overlay, widgetCount=5}
 widgets [
  Overlay_0 {widgetClassPath=&Overlay, widget=Overlay_0}
  Image_BGSpacer {
   widgetClassPath=&Image
   bIsVariable=true
   parent=Overlay_0
   widget=Image_BGSpacer
   properties {colorAndOpacity=<rgba 0 0 0 0>}
   slot {class=OverlaySlot, properties {horizontalAlignment=HAlign_Fill, verticalAlignment=VAlign_Fill}}
  }
  CommonTextBlock_65 {
   widgetClassPath=&CommonTextBlock
   parent=Overlay_0
   widget=CommonTextBlock_65
   properties {
    style=&TextStyle-Base_C
    text=READY UP SCREEN
    visibility=Collapsed
    font {
     bForceMonospaced=false
     bMaterialIsStencil=false
     fontMaterial=null
     fontObject=&NotoSans
     letterSpacing=0
     monospacedWidth=1
     size=24
     skewAmount=0
     typefaceFontName=NotoSans-Regular
     outlineSettings {
      bApplyOutlineToDropShadows=false
      bMiteredCorners=false
      bSeparateFillAlpha=false
      outlineMaterial=null
      outlineSize=0
      outlineColor=<rgba 0 0 0 1>
     }
    }
    shadowOffset=<0 0>
   }
   slot {class=OverlaySlot, properties {horizontalAlignment=HAlign_Fill, verticalAlignment=VAlign_Fill}}
  }
  ReadyUpOverlay {
   widgetClassPath=/Game/Athena/HUD/EndGame/IGM/ReadyUpScreenOverlay
   bInherited=true
   bIsVariable=true
   parent=Overlay_0
   widget=ReadyUpOverlay
   slot {class=OverlaySlot, properties {horizontalAlignment=HAlign_Fill, verticalAlignment=VAlign_Fill}}
  }
  KeepPlayingTogetherOverlay {
   widgetClassPath=/Game/Athena/HUD/EndGame/IGM/KeepPlayingTogetherOverlay
   bInherited=true
   bIsVariable=true
   parent=Overlay_0
   widget=KeepPlayingTogetherOverlay
   slot {class=OverlaySlot, properties {horizontalAlignment=HAlign_Fill, verticalAlignment=VAlign_Fill}}
  }
 ]
}
