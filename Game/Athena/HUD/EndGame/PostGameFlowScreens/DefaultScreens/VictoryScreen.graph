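# Widget Blueprint for the default Victory Royale screen in the Athena end-game post-game flow.
# It derives from the native FortVictoryRoyaleScreen class in FortniteUI.
# It adds no widgets, variables or event graph logic of its own, so any behavior comes from that parent class.
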WidgetBlueprint /Game/Athena/HUD/EndGame/PostGameFlowScreens/DefaultScreens/VictoryScreen
graphs [EventGraph {links [], nodes []}]
variables []
widgets {info {inheritedWidgetCount=0, parentClass=/Script/FortniteUI.FortVictoryRoyaleScreen, rootWidgetClass=null, widgetCount=0}, widgets []}
